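; Blueprint camera modifier for the Sideways gameplay, named for the second player in split screen, that applies a post process effect.
; Its BlueprintModifyPostProcess override returns a blend weight of 1.0 together with settings from GetPostPRocessSettings.
; The effect is applied at full strength on every update, with no use of delta time or fading in this graph.
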(fn BlueprintModifyPostProcess
  (params (DeltaTime :type Float%20%28single-precision%29))
  (return 1.0 (Variables|Default|GetPostPRocessSettings)))
