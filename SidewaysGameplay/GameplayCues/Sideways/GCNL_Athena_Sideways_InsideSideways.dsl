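; Looping gameplay cue for being inside the Sideways, handling splitscreen camera modifiers and a thermal material effect.
; With multiple local players, start adds a P1 or P2 camera modifier to the target's controller and stores it for removal to clear.
; A timeline, marked as repurposable leftover, writes thermal scalars to the global ThermalGlobal parameter collection.
; The play rate is set on Timeline_0, while Timeline_1 is played.
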(event EventOnLoopingStart(Cascade)
  (params (MyTarget :type Actor%20Object%20Reference) (Parameters :type Gameplay%20Cue%20Parameters%20Structure%20%28by%20ref%29) (ParticleComponents :type Array%20of%20Cascade%20Particle%20System%20Component%20Object%20References) (AudioComponents :type Array%20of%20Audio%20Component%20Object%20References))
  ;[1 1 1] Add Camera modifiers if the player is in splitscreen
  (if (Viewport|HasMultipleLocalPlayers 0)
    (Utilities|IsFirst MyTarget
      (:true
        (bind _returnvalue (Fortnite|Camera|AddCameraModifierforOwningPlayerController (Variables|Splitscreen|GetSplitScreenCameraModifier_P1) MyTarget))
        (Variables|Splitscreen|SetFirstPost _returnvalue))
      (:false
        (bind _returnvalue_2 (Fortnite|Camera|AddCameraModifierforOwningPlayerController (Variables|Splitscreen|GetSplitScreenCameraModifier_P2) MyTarget))
        (Variables|Splitscreen|SetSecondPost _returnvalue_2)))
    ;[1 0.75 0] This is some junk to make the thermal post function feel free to repurpose if needed or delete if not needed.
    (Components|Timeline|SetPlayRate (Variables|GCNL_Athena_Sideways_InsideSideways|GetTimeline_0) (Variables|Default|GetFadeInPlayrate))
    (bind (_direction _thermalramp) (Timeline|Timeline_1 PlayFromStart)
      (:Update
        (if (not (Utilities|Enum|Equal(Enum) _direction "Backward"))
          (Rendering|Material|SetScalarParameterValue 0 "/Game/Athena/Items/Weapons/Abilities/Thermal/MPC_ThermalGlobal.MPC_ThermalGlobal" "ThermalFadeout" _thermalramp :!class /Script/Engine.KismetMaterialLibrary)
          (Rendering|Material|SetScalarParameterValue 0 "/Game/Athena/Items/Weapons/Abilities/Thermal/MPC_ThermalGlobal.MPC_ThermalGlobal" "ThermalScale" 1.0 :!class /Script/Engine.KismetMaterialLibrary))))))

(event EventOnRemoval(Cascade)
  (params (MyTarget :type Actor%20Object%20Reference) (Parameters :type Gameplay%20Cue%20Parameters%20Structure%20%28by%20ref%29) (ParticleComponents :type Array%20of%20Cascade%20Particle%20System%20Component%20Object%20References) (AudioComponents :type Array%20of%20Audio%20Component%20Object%20References) (BurstCameraShakeInstance :type Camera%20Shake%20Base%20Object%20Reference) (BurstDecalInstance :type Decal%20Actor%20Object%20Reference))
  ;[1 1 1] Add Camera modifiers if the player is in splitscreen
  (if (Viewport|HasMultipleLocalPlayers 0)
    (bind _asplayer_controller (Utilities|IsFirst MyTarget)
      (:true
        (Game|Player|RemoveCameraModifier (Variables|/Script/Engine.PlayerController|GetPlayerCameraManager _asplayer_controller) (Variables|Splitscreen|GetFirstPost)))
      (:false
        (Game|Player|RemoveCameraModifier (Variables|/Script/Engine.PlayerController|GetPlayerCameraManager _asplayer_controller) (Variables|Splitscreen|GetSecondPost))))))

(fn ConstructionScript)
